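; Blueprint for a ceiling variant of a barrier trap device in Creative, deriving most behavior from its parent class.
; It overrides GetTriggerLocation to keep the Trigger box's relative X and Y while setting Z to its negated Z extent times the user's ZoneZScale option, so the zone offsets downward as it scales.
; Construction and placement only call the parent, and BeginPlay is a disabled stub.
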(event EventBeginPlay
  (|CallParentFunction) ; This node is disabled and will not be called.\nDrag off pins to build functionality.
)

(event EventOnPlaced
  (|CallParentFunction))

(fn GetTriggerLocation
  (return (Math|Vector|MakeVector (.x (Variables|/Script/Engine.SceneComponent|GetRelativeLocation (Variables|Default|GetTrigger))) (.y (Variables|/Script/Engine.SceneComponent|GetRelativeLocation (Variables|Default|GetTrigger))) (Math|Float|float*float (.z (Variables|/Script/Engine.BoxComponent|GetBoxExtent (Variables|Default|GetTrigger))) (Variables|UserOptions|GetZoneZScale) -1.0))))

(fn ConstructionScript
  (|CallParentFunction))
