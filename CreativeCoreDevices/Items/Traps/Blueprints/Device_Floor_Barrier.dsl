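; Creative floor barrier device that provides a resizable, shape-selectable collision volume, optionally blocking weapons, toggled by enable/disable events or minigame phase.
; StateLogic drives SetBarrier, which updates size, enabled state and navigation on authority, and can send a destruction event.
; Volume updates async-load mesh and material, then scale bounds, trigger and frame.
; Outside running minigames, an overlap volume fades visuals for the local player inside, using client-only cosmetics.
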(event EventEnableEvent
  (params (Instigator :type Controller%20Object%20Reference))
  ;[0.6648 0 1] Transmitters and receivers
  (|CallParentFunction Instigator)
  (CallFunction|SetBarrier true))

(event EventDisableEvent
  (params (Instigator :type Controller%20Object%20Reference))
  ;[0.6648 0 1] Transmitters and receivers
  (|CallParentFunction Instigator)
  (CallFunction|SetBarrier false))

(event EventBeginPlay
  ;[0 1 0.1154] Initialization
  (|CallParentFunction)
  (Networking|SwitchHasAuthority
    (:Authority
      (Collision|SetCollisionEnabled (Variables|Default|GetFxBarrierOverlapVolume) "NoCollision"))))

(event EventOnPlaced
  ;[0 1 0.1154] Initialization
  (Trap|AddTriggerComponent (Variables|Default|GetTrigger))
  (Variables|Default|SetAllowBoundsTransition true)
  (Networking|CanExecuteCosmeticEvents
    (:True
      (Audio|PlaySoundatLocation "/Game/Sounds/Creative/Gadgets/Traps/Trap_Elimination_Placed_Cue.Trap_Elimination_Placed_Cue" (Transformation|GetActorLocation)))))

(event OnAnyPropertyChanged(ToyOptionsComponent)
  ;[0 1 0.1154] Initialization
  (CallFunction|UpdateBlockingChannels)
  (CallFunction|UpdateVolumeSize)
  (Networking|SwitchHasAuthority
    (:Authority
      (CallFunction|StateLogic "PreGame"))
    (:Remote
      (bind _fortminigameprogress (Variables|Default|GetFortMinigameProgress)
        (:then
          (bind _currentminigame (Variables|/Script/FortniteGame.FortMinigameLogicComponent|GetCurrentMinigame _fortminigameprogress)
            (:then
              (CallFunction|StateLogic (Variables|/Script/FortniteGame.FortMinigame|GetCurrentState _currentminigame)))
            (:else)))
        (:else)))))

(event OnPropertiesLoaded(ToyOptionsComponent)
  ;[0 1 0.1154] Initialization
  (CallFunction|UpdateBlockingChannels)
  (CallFunction|UpdateVolumeSize)
  (Networking|SwitchHasAuthority
    (:Authority
      (CallFunction|StateLogic "PreGame"))
    (:Remote
      (bind _fortminigameprogress (Variables|Default|GetFortMinigameProgress)
        (:then
          (bind _currentminigame (Variables|/Script/FortniteGame.FortMinigameLogicComponent|GetCurrentMinigame _fortminigameprogress)
            (:then
              (CallFunction|StateLogic (Variables|/Script/FortniteGame.FortMinigame|GetCurrentState _currentminigame)))
            (:else)))
        (:else)))))

(event EventMinigameStateLogic
  (params (State :type EFortMinigameState%20Enum :default "PreGame"))
  ;[0 0.432 1] Minigame State Change Logic
  (CallFunction|StateLogic State))

(event OnComponentBeginOverlap(FxBarrierOverlapVolume)
  (params (OverlappedComponent :type Primitive%20Component%20Object%20Reference) (OtherActor :type Actor%20Object%20Reference) (OtherComp :type Primitive%20Component%20Object%20Reference) (OtherBodyIndex :type Integer) (bFromSweep :type Boolean) (SweepResult :type Hit%20Result%20Structure%20%28by%20ref%29))
  ;[1 1 1] Overlap Effects (local client only)
  (ClientVisuals|FadeonOverlap false OtherActor))

(event EventUpdateVolumeState
  ;[1 1 1] Updating Volume State (Override from Parent)
  (Collision|SetCollisionEnabled (Variables|Default|GetBarrierBounds) (select (Variables|Default|GetbEnabled) "QueryAndPhysics" "NoCollision"))
  (Networking|CanExecuteCosmeticEvents
    (:True
      (Collision|SetCollisionResponsetoChannel (Variables|Default|GetFxBarrierOverlapVolume) "ECC_Pawn" (select (Variables|Default|GetMinigameRunning) "ECR_Ignore" "ECR_Overlap"))
      (CallFunction|UpdateBoundsVisibility)
      (CallFunction|UpdateFade (Variables|Default|GetbEnabled)))))

(event OnComponentEndOverlap(FxBarrierOverlapVolume)
  (params (OverlappedComponent :type Primitive%20Component%20Object%20Reference) (OtherActor :type Actor%20Object%20Reference) (OtherComp :type Primitive%20Component%20Object%20Reference) (OtherBodyIndex :type Integer))
  ;[1 1 1] Overlap Effects (local client only)
  (ClientVisuals|FadeonOverlap true OtherActor))

(event Custom|Fade%20Barrier%20Bounds%20Visibility
  (params (On :type Boolean) (Immediate :type Boolean))
  ;[1 0.7982 0] Fade Visual Bound Visibility
  (Networking|CanExecuteCosmeticEvents
    (:True
      (if (not Immediate)
        (Development|SetHiddeninGame (Variables|Default|GetBarrierBounds) (not (Utilities|BarrierBoundsNotDisabled)) false)
        (if On
          (bind (_direction _lerp) (Timeline|FadeBoundsTL Reverse :NewTime (Math|Float|SelectFloat 0.0 (Components|Timeline|GetTimelineLength (Variables|Device_Floor_Barrier|GetFadeBoundsTL)) On))
            (:Update
              (Rendering|Material|SetScalarParameterValue (Variables|Default|GetBoundsMID) "Opacity" _lerp :!class /Script/Engine.MaterialInstanceDynamic))
            (:Finished
              (switch Utilities|FlowControl|Switch|SwitchonETimelineDirection _direction
                (:Forward
                  (Development|SetHiddeninGame (Variables|Default|GetBarrierBounds) true false))))) ; 0s = Full Opacity\n0.3s (End) = No Opacity
          (else
            (Timeline|FadeBoundsTL Play)))
        (else
          (Development|SetHiddeninGame (Variables|Default|GetBarrierBounds) (not (and On (Utilities|BarrierBoundsNotDisabled))) false)
          (Rendering|Material|SetScalarParameterValue (Variables|Default|GetBoundsMID) "Opacity" (Math|Float|SelectFloat 0.0 1.0 On) :!class /Script/Engine.MaterialInstanceDynamic)
          (Timeline|FadeBoundsTL SetNewTime))))))

(event Custom|SetBoundsVisuals
  (bind _zonevisibiltyindex (Variables|UserOptions|GetZoneVisibiltyIndex))
  (bind _barrier_materials (Variables|Default|GetBarrier%20Materials))
  (bind _output (Utilities|Array|Get(acopy) _barrier_materials _zonevisibiltyindex))
  ;[1 0.6163 0] Asynchronously Load Material for the Barrier Bounds, then update visuals
  (if (and (> (Variables|UserOptions|GetZoneVisibiltyIndex) 1) (Utilities|Array|IsValidIndex (Variables|Default|GetBarrier%20Materials) (Variables|UserOptions|GetZoneVisibiltyIndex)))
    (bind _loaded_object (Creative|Loading|AsyncLoadEditorSafe _output)
      (:Loaded
        (bind _asmaterial_interface (Utilities|Casting|CastToMaterialInterface _loaded_object))
        (bind _output_get (Variables|Default|SetBoundsMID (Material|GetMIDfromMaterial (Variables|Default|GetBarrierBounds) 0 _asmaterial_interface)))
        (bind _returnvalue (Rendering|Material|GetBaseMaterial _output_get))
        (Variables|Default|SetisOpaqueBoundsMaterial (and (not (Utilities|Array|ContainsItem (Variables|Default|GetNon-Shadowed%20Materials) _output)) (or (Utilities|Enum|Equal(Enum) (Variables|/Script/Engine.Material|GetBlendMode _returnvalue) "BLEND_Opaque") (Utilities|Enum|Equal(Enum) (Variables|/Script/Engine.Material|GetBlendMode _returnvalue) "BLEND_Masked"))))))
    (else
      (Variables|Default|SetBoundsMID 0)
      (Variables|Default|SetisOpaqueBoundsMaterial false)))
  (CallFunction|UpdateBoundsVisibility))

(event EventUpdateVolumeSize
  (|CallParentFunction)
  (Variables|Default|Setlocal%20Scale (Math|Vector|MakeVector (Variables|UserOptions|GetZoneXScale) (Variables|UserOptions|GetZoneYScale) (Variables|UserOptions|GetZoneZScale)))
  (Utilities|FlowControl|Sequence
    (:then_0
      (Networking|CanExecuteCosmeticEvents
        (:True
          (CallFunction|SetBoundsVisuals))))
    (:then_1
      (CallFunction|SetBarrierMesh) ; async loads the appropriate Barrier mesh \nthen sets the scale
      (Transformation|SetWorldScale3D (Variables|Default|GetTrigger) (Variables|Default|Getlocal%20Scale))
      (Components|Box|SetBoxExtent (Variables|Default|GetFxBarrierOverlapVolume) (+ (* (Variables|Default|Getlocal%20Scale) (Variables|Default|GetFXBarrierBoxExtent)) (Variables|Default|GetFXBarrierVolumeBias)) true)
      (bind _location (CallFunction|GetTriggerLocation))
      (Transformation|SetRelativeLocation (Variables|Default|GetFxBarrierOverlapVolume) _location false true)
      (ClientVisuals|ScaleFrame))))

(event Custom|Set%20Barrier%20Mesh
  (bind _knob_volumeshape (Variables|UserOptions|Basic|GetKnob_VolumeShape))
  (bind _volume_shapes (Variables|Default|GetVolume%20Shapes))
  (bind (_value _returnvalue) (Utilities|Map|Find _volume_shapes _knob_volumeshape))
  ;[0 0.2775 2] Set Mesh
  (if (and (Utilities|IsValidSoftObjectReference _value) _returnvalue)
    (bind _loaded_object (Creative|Loading|AsyncLoadEditorSafe _value)
      (:Loaded
        (bind _asstatic_mesh (Utilities|Casting|CastToStaticMesh _loaded_object))
        (Components|StaticMesh|SetStaticMesh (Variables|Default|GetBarrierBounds) _asstatic_mesh)
        (CallFunction|SetBarrierScale)))))

(fn FadeOnOverlap
  (params (Fade%20Up :type Boolean) (Actor :type Actor%20Object%20Reference))
  (LocalVariables|Setlocal%20Fade%20Up (or Fade%20Up (Utilities|Enum|Equal(Enum) (Variables|UserOptions|Basic|GetKnob_VolumeShape) "NewEnumerator2")))
  (Networking|CanExecuteCosmeticEvents
    (:True
      (bind _aspawn (Utilities|Casting|CastToPawn Actor))
      (if (Pawn|IsLocallyControlled _aspawn)
        (Variables|Default|SetLocalPlayerIsWithinFxVolume (not (LocalVariables|Getlocal%20Fade%20Up)))
        (if (and (>= (Variables|UserOptions|GetZoneVisibiltyIndex) 2) (Variables|Default|GetAllowBoundsTransition))
          (CallFunction|FadeBarrierBoundsVisibility (LocalVariables|Getlocal%20Fade%20Up)))))))

(fn GetBoundsComponents
  (params (SelfActor :type Actor%20Object%20Reference))
  (return (Utilities|Array|MakeArray (Variables|Debug|GetStaticMeshComponent))))

(fn IsVisibleDuringMinigame
  (return (and (|CallParentFunction) (not (Variables|Default|GetisOpaqueBoundsMaterial)))))

(fn IsHollowVolumeShape
  (params (VolumeShape :type EGameplayVolumeShape%20Enum :default "NewEnumerator0"))
  (return (Utilities|Array|ContainsItem (Variables|Default|GetHollowVolumeShapes) VolumeShape)))

(fn ScaleFrame
  (Rendering|Material|SetVectorParameterValue (Material|GetMID (Variables|Default|GetFrame)) "SizeBias" (Math|Conversions|ToLinearColor(Vector) (- (* (Math|Vector|MakeVector 256.0 256.0 384.0) (Math|Vector|MakeVector (Variables|UserOptions|GetZoneXScale) (Variables|UserOptions|GetZoneYScale) (Variables|UserOptions|GetZoneZScale))) (Math|Vector|MakeVector 256.0 256.0 384.0))) :!class /Script/Engine.MaterialInstanceDynamic)
  (Rendering|SetBoundsScale (Variables|Default|GetFrame) (* (Math|Float|Max(Float) (Variables|UserOptions|GetZoneXScale) (Variables|UserOptions|GetZoneYScale) (Variables|UserOptions|GetZoneZScale)) 1.5)))

(fn SetBarrier
  (params (On :type Boolean))
  (bind _on (Variables|GetOn))
  (CallFunction|UpdateVolumeSize)
  (Networking|SwitchHasAuthority
    (:Authority
      (Utilities|FlowControl|Sequence
        (:then_0
          (if (Math|Boolean|ANDBoolean (Variables|Default|GetMinigameRunning) (Variables|GetOn) (Utilities|Enum|NotEqual(Enum) (Variables|UserOptions|Basic|GetKnob_VolumeShape) "NewEnumerator2"))
            (Ability|SendGameplayEventtoActor self (Variables|Default|GetActor%20Destruction%20Tag))))
        (:then_1
          (Variables|Default|SetbEnabled _on)
          (Navigation|UpdateNavigationProperties 0 _on))))
    (:Remote
      ;[1 1 1] BarrierOn's not going to trigger the repnotify, let's call the update function manually
      (if (Math|Boolean|Equal(Boolean) (Variables|Default|GetbEnabled) (Variables|GetOn))
        (CallFunction|UpdateVolumeState)))))

(fn SetBarrierScale
  (bind _knob_volumeshape (Variables|UserOptions|Basic|GetKnob_VolumeShape))
  (bind _volume_offsets (Variables|Default|GetVolume%20Offsets))
  (bind (_value _returnvalue) (Utilities|Map|Find _volume_offsets _knob_volumeshape))
  (bind (_barrier_scale_x _barrier_scale_y _barrier_scale_z) (Math|Vector|BreakVector (Variables|GetBarrier%20Scale_X)))
  (Utilities|FlowControl|Sequence
    (:then_0
      (LocalVariables|SetBarrier%20Scale (- (* (Variables|Default|Getlocal%20Scale) (Variables|Default|GetFXBarrierGridScale)) (LocalVariables|GetBarrier%20Shrink))))
    (:then_1
      (Transformation|SetWorldScale3D (Variables|Default|GetBarrierBounds) (LocalVariables|GetBarrier%20Scale)))
    (:then_2
      (if _returnvalue
        (Transformation|SetRelativeLocation (Variables|Default|GetBarrierOffset) (Math|Vector|MakeVector 0.0 0.0 (* _value _barrier_scale_z)) false false)))))

(fn StateLogic
  (params (State :type EFortMinigameState%20Enum :default "PreGame"))
  (if (Math|Boolean|ORBoolean (Utilities|Enum|Equal(Enum) "PreGame" State) (and (Utilities|Enum|Equal(Enum) "Transitioning" State) (Class|CreativeCommonDeviceFunctionLibrary|IsonUnpublishedIsland self)) (Utilities|Enum|Equal(Enum) "InProgress" State))
    (Variables|Default|SetMinigameRunning (or (not (Class|CreativeCommonDeviceFunctionLibrary|IsonUnpublishedIsland self)) (Utilities|Enum|NotEqual(Enum) "PreGame" (Variables|GetState))))
    (CallFunction|SetBarrier (Class|CreativeCommonDeviceFunctionLibrary|IsEnabledPhaseMatch (Variables|GetState) (Variables|UserOptions|GetEnabledIndex) self))))

(fn UpdateBlockingChannels
  (bind _blockweaponsingame (Variables|UserOptions|GetBlockWeaponsInGame))
  (bind _returnvalue (select _blockweaponsingame "ECR_Block" "ECR_Ignore"))
  (bind _barrierbounds (Variables|Default|GetBarrierBounds))
  (Collision|SetCollisionResponsetoChannel _barrierbounds "ECC_GameTraceChannel7" _returnvalue)
  (Collision|SetCollisionResponsetoChannel _barrierbounds "ECC_GameTraceChannel5" _returnvalue)
  (Collision|SetCollisionResponsetoChannel _barrierbounds "ECC_GameTraceChannel3" _returnvalue)
  (Collision|SetCollisionResponsetoChannel _barrierbounds "ECC_GameTraceChannel8" _returnvalue)
  (Collision|SetCollisionResponsetoChannel _barrierbounds "ECC_GameTraceChannel12" _returnvalue))

(fn UpdateBoundsVisibility
  (bind _returnvalue (Utilities|BarrierBoundsNotDisabled))
  (Development|SetHiddeninGame (Variables|Default|GetBarrierBounds) (not (and _returnvalue (Math|Boolean|ORBoolean (not (Variables|Default|GetLocalPlayerIsWithinFxVolume)) (Variables|Default|GetMinigameRunning) (Utilities|Enum|Equal(Enum) (Variables|UserOptions|Basic|GetKnob_VolumeShape) "NewEnumerator2")))) false)
  (Rendering|SetCastShadow (Variables|Default|GetBarrierBounds) (and (Variables|Default|GetisOpaqueBoundsMaterial) _returnvalue))
  (return _returnvalue))

(fn UpdateFade
  (params (On :type Boolean) (Immediate :type Boolean))
  (Networking|CanExecuteCosmeticEvents
    (:True
      (Variables|Default|SetAllowBoundsTransition On)
      (CallFunction|FadeBarrierBoundsVisibility (and On (or (not (Variables|Default|GetLocalPlayerIsWithinFxVolume)) (Variables|Default|GetMinigameRunning))) Immediate) ; Only Fade if player is outwith bounds \n(If minigame isn't running)
    )))

(fn UpdateNavigationProperties
  (params (bBarrierEnabled :type Boolean))
  (CallFunction|SetIsNavigationRelevant bBarrierEnabled)
  (Navigation|SetAlwaysExportNavRelevantComponent bBarrierEnabled)
  (Navigation|SetNavigationObstacleOverride (select (CallFunction|IsHollowVolumeShape (Variables|UserOptions|Basic|GetKnob_VolumeShape)) "ExportAsPrimitive" "ForceEnabled")))

(fn ConstructionScript
  (|CallParentFunction)
  (CallFunction|SetBoundsVisuals))
